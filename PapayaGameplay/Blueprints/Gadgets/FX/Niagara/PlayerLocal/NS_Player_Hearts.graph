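# Music-reactive player gadget Niagara effect, named for hearts, that emits rising sprites along bones of the user-supplied skeletal mesh.
# Sprites spawns continuously and SpritesBurst fires bursts with probability set by jukebox average amplitude, both placed by FN_BoneFlow with inherited velocity.
# Per-particle FN_SoundDt detects jukebox percussion hits and feeds a decaying accumulator.
# That accumulator modulates brightness, size and drag, plus curl noise on Sprites only.
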NiagaraSystem /PapayaGameplay/Blueprints/Gadgets/FX/Niagara/PlayerLocal/NS_Player_Hearts
dependencies {usedDataInterfaces [&NiagaraDataInterfaceCurve], usedDynamicInputs [], usedModules [&SystemState]}
emitters {
 Sprites {
  emitterUpdateScript [
   &EmitterState(
    Life Cycle Mode=::NewEnumerator1
    Loop Duration=&UniformRangedFloat(Minimum=1, Maximum=2)
    Recalculate Duration Each Loop=true
    UseLoopDelay=false
   )
   &SpawnRate(SpawnRate=10, Use Spawn Probability=false)
  ]
  particleSpawnScript [
   /Niagara/Modules/Spawn/Initialization/InitializeParticle(
    Color=&MakeLinearColorFromVectorAndFloat("Vector (RGB)"=&UniformRangedVector(Minimum=<1 1 1>, Maximum=<5 2 1>), "Float (Alpha)"=10)
    Lifetime=&UniformRangedFloat(Minimum=0.8, Maximum=2.5)
    Module.Write Color=true
    Module.Write Lifetime=true
    Module.Write Mass=true
    Module.Write Material Random=true
    Module.Write Position=false
    Module.Write Scale=false
    Module.Write SpriteRotation=true
    Module.Write SpriteSize=true
    Position=&SimulationPosition()
    Sprite Size=&Vector2DFromFloat(Value=&Multiply_Float(A=&UniformAOrBFloat(A=0.5), B=10))
   )
   /SetVariables(Particles.Velocity=&Multiply_Vector(A=$Output.None.RotatedBiTangent, B=&VectorFromFloat(Value=0.125))) {isSetParameters=true, module=SetVariables_#1}
   /Game/Effects/Niagara/Systems/Prototype/Tests/MusicDrivenCCPM/FN_BoneFlow(
    Angle:NiagaraFloat=&UniformRangedFloat(Minimum=0, Maximum=180)
    BoneID=HlslExpression() {hlslExpression="(Particles.UniqueID && 1) *3"}
    Radius:NiagaraFloat=&FloatFromCurve(
     FloatCurve=DataInterface() {propertyValues {Curve {keys [<at -0.0046 1.0113>,<at 1.0011 0.1>]}}}
     CurveIndex=$Particles.MaterialRandom
     Scale Curve=10
    )
    SkeletalMesh:NiagaraDataInterfaceSkeletalMesh=$User.SkeletalMesh
    T:NiagaraFloat=$Particles.MaterialRandom
   )
   &InheritVelocity(Module.Limit Speed=true, Velocity Limit=350, Velocity Scale=<2 2 2>)
  ]
  particleUpdateScript [
   /Game/Effects/Niagara/Systems/Prototype/Tests/MusicDrivenCCPM/FN_SoundDt(FFT:NiagaraFloat=&MakeFloatFromVector4(Vector 4=$NPC.Papaya_Jukebox_NPC.Percussion, Channel=::NewEnumerator0))
   /Game/Effects/Niagara/Modules/Math/FN_Accumulator(DecayRate=0.1, Input=HlslExpression() {hlslExpression="((Particles.FN_SoundDt.Dt > .1) &&&&( Particles.FN_SoundDt.Sample > .4 ))? 1. : 0"})
   /Game/Effects/Niagara/Modules/Math/FN_Integrator/FN_Integrator(Input:NiagaraFloat=$Particles.FN_Accumulator.Attr)
   &ParticleState()
   &ScaleColor(
    Scale Alpha=&FloatFromCurve(FloatCurve=DataInterface() {propertyValues {Curve {keys [
         {arriveTangent=6.0619, interpMode=RCIM_Cubic, leaveTangent=6.0619, tangentMode=RCTM_User, time=0, value=0}
         {arriveTangent=0.002, interpMode=RCIM_Cubic, leaveTangent=0.002, time=0.5063, value=0.9903}
         <at 1.0011 0.002>
        ]}}})
    Scale RGB=&Lerp_Vector(A=<0.8 0.8 0.8>, B=<1 1 1>, Alpha=$Particles.FN_Accumulator.Attr)
    ScaleA=true
    ScaleRGB=true
   )
   &Drag(Drag=&Lerp_Float(A=5, B=0, Alpha=$Particles.FN_Accumulator.Attr), Use Linear Drag=true, Use Rotational Drag=false)
   &AccelerationForce(Acceleration=&MakeVector(X=0, Y=0, Z=35))
   /Niagara/Modules/Update/Forces/CurlNoiseForce(Noise Frequency=25, Noise Strength=&Lerp_Float(A=0, B=500, Alpha=$Particles.FN_Accumulator.Attr), Pan Noise=false)
   &SolveForcesAndVelocity(Clamp Velocity=false, Limit Acceleration=false)
   &CameraOffset(Camera Offset Amount=&Lerp_Float(A=0, B=25, Alpha=$Particles.NormalizedAge))
   &ScaleSpriteSize(Scale Factor=&Vector2DFromFloat(Value=&FloatFromCurve(
      FloatCurve=DataInterface() {propertyValues {Curve {keys [
          <at 0 0>
          {arriveTangent=0.002, interpMode=RCIM_Cubic, leaveTangent=0.002, time=0.1088, value=1}
          {arriveTangent=-2.3286, interpMode=RCIM_Cubic, leaveTangent=-2.3286, tangentMode=RCTM_User, time=1.0011, value=0.002}
         ]}}}
      Scale Curve=&Multiply_Float(A=1, B=&Add_Float(A=$Particles.FN_Accumulator.Attr, B=1))
     )))
   &DynamicMaterialParameters(
    Index 0 Param 1=&Lerp_Float(A=1, B=0, Alpha=$Particles.NormalizedAge)
    Index 0 Param 2=&FloatFromCurve(FloatCurve=DataInterface() {propertyValues {Curve {keys [<at 0 1 RCIM_Cubic>, {arriveTangent=-3.0397, interpMode=RCIM_Cubic, leaveTangent=-3.0397, tangentMode=RCTM_User, time=1, value=0}]}}})
    Index 0 Param 3=1
    Module.Param0WriteEnabled=true
    Module.Param1WriteEnabled=true
    Module.Param2WriteEnabled=true
    Module.Param3WriteEnabled=false
   )
  ]
  rendererClasses [&NiagaraSpriteRendererProperties]
  renderers [{rendererClass=&NiagaraSpriteRendererProperties}]
 }
 SpritesBurst {
  emitterUpdateScript [
   &EmitterState(Life Cycle Mode=::NewEnumerator1, Loop Duration=1, UseLoopDelay=false)
   &SpawnBurst_Instantaneous(Spawn Count=50, Spawn Probability=$NPC.Papaya_Jukebox_NPC.AverageAmplitude, Use Spawn Probability=true)
  ]
  particleSpawnScript [
   &InitializeParticle(
    Color=&MakeLinearColorFromVectorAndFloat("Vector (RGB)"=&UniformRangedVector(Minimum=<1 1 1>, Maximum=<5 2 1>), "Float (Alpha)"=10)
    Lifetime=&UniformRangedFloat(Minimum=0.1, Maximum=0.6)
    Module.Write Color=true
    Module.Write Lifetime=true
    Module.Write Mass=true
    Module.Write Material Random=true
    Module.Write Position=false
    Module.Write Scale=false
    Module.Write SpriteRotation=true
    Module.Write SpriteSize=true
    Position=&SimulationPosition()
    Sprite Rotation=&UniformRangedFloat(Minimum=0, Maximum=1)
    Sprite Size=&Vector2DFromFloat(Value=&Multiply_Float(A=&UniformAOrBFloat(A=0.2, B=0.4), B=8))
   )
   /SetVariables(Particles.Velocity=&Multiply_Vector(A=$Output.None.RotatedBiTangent, B=&VectorFromFloat(Value=3))) {isSetParameters=true, module=SetVariables_#1}
   &FN_BoneFlow(
    Angle:NiagaraFloat=&UniformRangedFloat(Minimum=0, Maximum=180)
    BoneID=HlslExpression() {hlslExpression="(Particles.UniqueID && 1) *3"}
    Radius:NiagaraFloat=&FloatFromCurve(
     FloatCurve=DataInterface() {propertyValues {Curve {keys [<at -0.0046 1.0113>,<at 1.0011 0.1>]}}}
     CurveIndex=$Particles.MaterialRandom
     Scale Curve=10
    )
    SkeletalMesh:NiagaraDataInterfaceSkeletalMesh=$User.SkeletalMesh
    T:NiagaraFloat=$Particles.MaterialRandom
   )
   &InheritVelocity(Module.Limit Speed=true, Velocity Limit=350, Velocity Scale=<2 2 2>)
  ]
  particleUpdateScript [
   &FN_SoundDt(FFT:NiagaraFloat=&MakeFloatFromVector4(Vector 4=$NPC.Papaya_Jukebox_NPC.Percussion, Channel=::NewEnumerator0))
   &FN_Accumulator(DecayRate=0.1, Input=HlslExpression() {hlslExpression="((Particles.FN_SoundDt.Dt > .1) &&&&( Particles.FN_SoundDt.Sample > .4 ))? 1. : 0"})
   &FN_Integrator(Input:NiagaraFloat=$Particles.FN_Accumulator.Attr)
   &ParticleState()
   &ScaleColor(
    Scale Alpha=&FloatFromCurve(FloatCurve=DataInterface() {propertyValues {Curve {keys [
         {arriveTangent=6.0619, interpMode=RCIM_Cubic, leaveTangent=6.0619, tangentMode=RCTM_User, time=0, value=0}
         {arriveTangent=0.002, interpMode=RCIM_Cubic, leaveTangent=0.002, time=0.5063, value=0.9903}
         <at 1.0011 0.002>
        ]}}})
    Scale RGB=&Lerp_Vector(A=<0.8 0.8 0.8>, B=<1 1 1>, Alpha=$Particles.FN_Accumulator.Attr)
    ScaleA=true
    ScaleRGB=true
   )
   &Drag(Drag=&Lerp_Float(A=0, B=2, Alpha=$Particles.FN_Accumulator.Attr), Use Linear Drag=true, Use Rotational Drag=false)
   &AccelerationForce(Acceleration=&MakeVector(X=0, Y=0, Z=15))
   &SolveForcesAndVelocity(Clamp Velocity=false, Limit Acceleration=false)
   &CameraOffset(Camera Offset Amount=&Lerp_Float(A=0, B=25, Alpha=$Particles.NormalizedAge))
   &ScaleSpriteSize(Scale Factor=&Vector2DFromFloat(Value=&FloatFromCurve(
      FloatCurve=DataInterface() {propertyValues {Curve {keys [<at 0 0>,<at 0.5914 1>,<at 1.0011 0.002>]}}}
      Scale Curve=&Multiply_Float(A=1, B=&Add_Float(A=$Particles.FN_Accumulator.Attr, B=1))
     )))
   &DynamicMaterialParameters(
    Index 0 Param 1=&Lerp_Float(A=0, B=1, Alpha=$Particles.NormalizedAge)
    Index 0 Param 2=1
    Index 0 Param 3=1
    Module.Param0WriteEnabled=true
    Module.Param1WriteEnabled=true
    Module.Param2WriteEnabled=true
    Module.Param3WriteEnabled=false
   )
  ]
  rendererClasses [&NiagaraSpriteRendererProperties]
  renderers [{rendererClass=&NiagaraSpriteRendererProperties}]
 }
}
summary {
 systemName=NS_Player_Hearts
 emitters [
  {emitterName=Sprites, rendererClasses [&NiagaraSpriteRendererProperties]}
  {emitterName=SpritesBurst, rendererClasses [&NiagaraSpriteRendererProperties]}
 ]
 userVariables [User.SkeletalMesh {
   type=NiagaraDataInterfaceSkeletalMesh
   defaultValue {
    struct=NiagaraExt_VariableValue_DataInterface
    value {dataInterface=&self:NiagaraDataInterfaceSkeletalMesh_0}
   }
  }]
}
